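# Niagara system for the Scarlet Bloom cosmetic shoes that sheds two-tone petals and draws running trails from the character's skeletal mesh.
# Idle and Run petal emitters trade spawn rates on the running flag, splitting particles between shoes via partitioned socket indices.
# Three layered anim-trail ribbon emitters show only while running outside skydiving.
# Per-shoe scale drives petal alpha and ribbon width, with time-of-day tinting or dimming.
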NiagaraSystem /CosmeticShoes/Assets/FORT_Shoes/Shoe_ScarletBloom/FX/NS_Shoe_ScarletBloom_Idle
dependencies {
 usedDataInterfaces [
  &NiagaraDataInterfaceArrayInt32
  &NiagaraDataInterfaceCurve
  &NiagaraDataInterfaceColorCurve
  &NiagaraDataInterfaceSkeletalMesh
 ]
 usedDynamicInputs [
  &Boolean_OrOperation
  /Game/Effects/Niagara/DynamicInputs/Logic/FN_IsRunningCheck
  &Divide_Float
 ]
 usedModules [
  &SystemState
  &self:SystemSpawnScript.SystemScriptSource.SystemScriptGraph.NiagaraNodeAssignment_4.SetVariables_#1
  &self:SystemSpawnScript.SystemScriptSource.SystemScriptGraph.NiagaraNodeAssignment_1.SetVariables_#2
  &self:SystemSpawnScript.SystemScriptSource.SystemScriptGraph.NiagaraNodeAssignment_2.SetVariables_#3
 ]
}
emitters {
 NE_Petals_Idle {
  emitterUpdateScript [
   &EmitterState(
    Inactive Response=::NewEnumerator1
    Life Cycle Mode=::NewEnumerator1
    Loop Behavior=::NewEnumerator1
    Loop Duration Mode=::NewEnumerator1
    UseLoopDelay=false
   )
   &SpawnRate(
    Spawn Probability=0.2
    SpawnRate=&Lerp_Float(A=2, B=0, Alpha=&MakeCustomFloatFromBool(Bool=$System.bIsRunning, True Float=1, False Float=0))
    Use Spawn Probability=true
   )
   &SpawnRate(
    Spawn Probability=0.2
    SpawnGroup=1
    SpawnRate=&Lerp_Float(A=2, B=0, Alpha=&MakeCustomFloatFromBool(Bool=$System.bIsRunning, True Float=1, False Float=0))
    Use Spawn Probability=true
   ) {module=SpawnRate001}
  ]
  particleSpawnScript [
   /SetVariables(Particles.VisibilityTag=&SpawnGroup()) {isSetParameters=true, module=SetVariables_#4}
   &PartitionParticles()
   &InitializeParticle(
    Color=&MergeHelper_ColorSwitch(0=<rgba 0.2 0 0.0166 1>, 1=<rgba 1 1 1 1>, 2=?, 3=?, Index=$Particles.VisibilityTag, Num=::NewEnumerator0)
    Lifetime Max=&Lerp_Float(A=3, B=1, Alpha=&MakeCustomFloatFromBool(Bool=$System.bIsRunning, True Float=1, False Float=0))
    Lifetime Min:NiagaraFloat=&Lerp_Float(A=2, B=0.5, Alpha=&MakeCustomFloatFromBool(Bool=$System.bIsRunning, True Float=1, False Float=0))
    Lifetime Mode=::NewEnumerator1
    Position Offset=&MergeHelper_VectorSwitch(
     0=<0 -5 -5>
     1=&MergeHelper_VectorSwitch(0=<10 -8 2>, 1=<-7 -10 5>, 2=?, 3=?, Index=&RandomRangeInt(Minimum=0, Maximum=1), Num=::NewEnumerator0)
     2=?
     3=?
     Index=$Particles.VisibilityTag
     Num=::NewEnumerator0
    )
    Position Offset Coordinate Space=::Local
    Sprite Rotation Mode=::NewEnumerator2
    Sprite Size Max=<5 5>
    Sprite Size Min=<3 3>
    Sprite Size Mode=::NewEnumerator2
    UsePositionOffset=true
    Write Material Random=true
   )
   &SkeletalMeshLocation(
    Bone / Socket Index=&SelectIntFromArray(Direct Array Index=$Particles.PartitionParticles.Partition, Int Selection Array=DataInterface() {propertyValues {IntData [2,3]}})
    MaskSpawnGroup=false
    Mesh Sampling Type=::NewEnumerator5
    Skeletal Mesh=$System.Skeletal Mesh
    Socket Sampling Mode=::NewEnumerator5
   )
  ]
  particleUpdateScript [
   &ParticleState()
   &GravityForce(
    Coordinate Space=::Local
    Gravity=&RandomRangeVector(
     Minimum=&MakeVector(
      X=0
      Y=&FloatFromCurve(
       FloatCurve=DataInterface() {propertyValues {Curve {
          defaultValue=1
          keys [{interpMode=RCIM_Cubic, tangentMode=RCTM_User, time=0, value=0}, {interpMode=RCIM_Cubic, tangentMode=RCTM_User, time=1, value=1}]
         }}}
       Scale Curve=-500
      )
      Z=-250
     )
     Maximum=&MakeVector(
      X=0
      Y=&FloatFromCurve(
       FloatCurve=DataInterface() {propertyValues {Curve {
          defaultValue=1
          keys [{interpMode=RCIM_Cubic, tangentMode=RCTM_User, time=0, value=0}, {interpMode=RCIM_Cubic, tangentMode=RCTM_User, time=1, value=1}]
         }}}
       Scale Curve=-750
      )
      Z=-500
     )
    )
   )
   &CurlNoiseForce(Bias Noise=false, Noise Frequency=5, Noise Strength=150, Pan Noise=false)
   &AerodynamicDrag(
    Rotation Strength:NiagaraFloat=&RandomRangeFloat(Minimum=5, Maximum=8)
    Use Linear Drag=true
    Use Rotational Drag=true
   )
   &SolveForcesAndVelocity(Clamp Velocity=false, Limit Acceleration=false)
   &ScaleSpriteSize(Uniform Curve Sprite Scale=DataInterface() {propertyValues {Curve {defaultValue=0, keys [{tangentMode=RCTM_User, time=0, value=0}, <at 0.1 1>, <at 0.5 1>, {tangentMode=RCTM_User, time=1, value=0}]}}})
   &ScaleColor(
    Scale Alpha=&FloatFromCurve(
     FloatCurve=DataInterface() {propertyValues {Curve {keys [<at 0 0>,<at 0.1 1>]}}}
     Scale Curve=&Lerp_Float(
      A=&Multiply_Float(
       A=1
       B=&MergeHelper_FloatSwitch(
        0=&MakeFloatFromVector(Vector=$User.ShoeScale_R, Channel=::NewEnumerator0)
        1=&MakeFloatFromVector(Vector=$User.ShoeScale_L, Channel=::NewEnumerator0)
        2=?
        3=?
        Index=$Particles.PartitionParticles.Partition
        Num=::NewEnumerator0
       )
      )
      B=0
      Alpha=&MakeCustomFloatFromBool(Bool=$User.IsSkydiving, True Float=1, False Float=0)
     )
    )
    Scale RGB=&MakeVectorFromLinearColorRGB(Color=&MergeHelper_ColorSwitch(
      0=&FNTimeOfDayColor(Night Color=<rgba 0.1 0 0.0083 1>, Morning Color=<rgba 0.4 0 0 1>, Day Color=<rgba 0.4 0 0 1>, Evening Color=<rgba 0.4 0 0 1>)
      1=&FNTimeOfDayColor(Night Color=<rgba 0.25 0.2917 0.5 1>, Morning Color=<rgba 1 1 1 1>, Day Color=<rgba 1 1 1 1>, Evening Color=<rgba 1 1 1 1>)
      2=?
      3=?
      Index=$Particles.VisibilityTag
      Num=::NewEnumerator0
     ))
    ScaleA=true
    ScaleRGB=true
   )
   &AlignSpriteToMeshOrientation(SetAlignment=true, SetFacing=true)
   &SpriteRotationRate(Rotation Rate=&RandomRangeFloat(Minimum=-15, Maximum=15))
   &CameraOffset(Camera Offset Amount=&FloatFromCurve(
     FloatCurve=DataInterface() {propertyValues {Curve {
        defaultValue=1
        keys [{interpMode=RCIM_Cubic, tangentMode=RCTM_User, time=0, value=0}, {interpMode=RCIM_Cubic, tangentMode=RCTM_User, time=1, value=1}]
       }}}
     Scale Curve=10
    ))
  ]
  rendererClasses [&NiagaraSpriteRendererProperties]
  renderers [
   {rendererClass=&NiagaraSpriteRendererProperties}
   {rendererClass=&NiagaraSpriteRendererProperties, rendererIndex=1}
  ]
 }
 NE_Petals_Run {
  emitterUpdateScript [
   &EmitterState(
    Inactive Response=::NewEnumerator1
    Life Cycle Mode=::NewEnumerator1
    Loop Behavior=::NewEnumerator1
    Loop Duration Mode=::NewEnumerator1
    UseLoopDelay=false
   )
   &SpawnRate(
    Spawn Probability=0.25
    SpawnRate=&Lerp_Float(
     A=0
     B=12
     Alpha=&MakeCustomFloatFromBool(
      Bool=&Boolean_AndOperation(Bool1=$System.bIsRunning, Bool2=&Boolean_NotOperation(Bool=$User.IsSkydiving))
      True Float=1
      False Float=0
     )
    )
    Use Spawn Probability=true
   )
   &SpawnRate(
    Spawn Probability=0.2
    SpawnGroup=1
    SpawnRate=&Lerp_Float(
     A=0
     B=12
     Alpha=&MakeCustomFloatFromBool(
      Bool=&Boolean_AndOperation(Bool1=$System.bIsRunning, Bool2=&Boolean_NotOperation(Bool=$User.IsSkydiving))
      True Float=1
      False Float=0
     )
    )
    Use Spawn Probability=true
   ) {module=SpawnRate001}
  ]
  particleSpawnScript [
   /SetVariables(Particles.VisibilityTag=&SpawnGroup()) {isSetParameters=true, module=SetVariables_#4}
   &PartitionParticles()
   &InitializeParticle(
    Color=&MergeHelper_ColorSwitch(0=<rgba 0.2 0 0.0166 1>, 1=<rgba 1 1 1 1>, 2=?, 3=?, Index=$Particles.VisibilityTag, Num=::NewEnumerator0)
    Lifetime Max=1
    Lifetime Min=0.5
    Lifetime Mode=::NewEnumerator1
    Position Offset=&MergeHelper_VectorSwitch(
     0=<0 -5 -5>
     1=&MergeHelper_VectorSwitch(0=<10 -8 2>, 1=<-7 -10 5>, 2=?, 3=?, Index=&RandomRangeInt(Minimum=0, Maximum=1), Num=::NewEnumerator0)
     2=?
     3=?
     Index=$Particles.VisibilityTag
     Num=::NewEnumerator0
    )
    Position Offset Coordinate Space=::Local
    Sprite Rotation Mode=::NewEnumerator2
    Sprite Size Max=<8 8>
    Sprite Size Min=<4 4>
    Sprite Size Mode=::NewEnumerator2
    UsePositionOffset=true
    Write Material Random=true
   )
   &SkeletalMeshLocation(
    Bone / Socket Index=&SelectIntFromArray(Direct Array Index=$Particles.PartitionParticles.Partition, Int Selection Array=DataInterface() {propertyValues {IntData [0,1]}})
    MaskSpawnGroup=false
    Mesh Sampling Type=::NewEnumerator5
    Skeletal Mesh=$System.Skeletal Mesh
    Socket Sampling Mode=::NewEnumerator5
   )
   &AddVelocity(
    Cone Axis=<0 1 0>
    Module.UseConeFalloff=false
    Module.UseVelDistribution=false
    Override Seed=false
    Velocity Mode=::NewEnumerator2
    Velocity Speed=&RandomRangeFloat(Minimum=-150, Maximum=-250)
   )
  ]
  particleUpdateScript [
   &ParticleState()
   &GravityForce(Coordinate Space=::Local, Gravity=&RandomRangeVector(Minimum=<0 0 -90>, Maximum=<0 0 -150>))
   &CurlNoiseForce(Bias Noise=false, Noise Frequency=5, Noise Strength=2.5, Pan Noise=false)
   &AerodynamicDrag(
    Rotation Strength:NiagaraFloat=&RandomRangeFloat(Minimum=5, Maximum=8)
    Use Linear Drag=true
    Use Rotational Drag=true
   )
   &InheritSourceMovement(
    Applied Position Scale=&VectorFromFloat(Value=&FloatFromCurve(FloatCurve=DataInterface() {propertyValues {Curve {keys [<at 0 0.5>,<at 1 0>]}}}))
    ApplyForce=false
    ApplyPosition=true
    ApplyPositionFalloff=false
   )
   &SolveForcesAndVelocity(Clamp Velocity=false, Limit Acceleration=false)
   &ScaleSpriteSize(Uniform Curve Sprite Scale=DataInterface() {propertyValues {Curve {defaultValue=0, keys [{tangentMode=RCTM_User, time=0, value=0}, <at 0.1 1>, <at 0.5 1>, {tangentMode=RCTM_User, time=1, value=0}]}}})
   &ScaleColor(
    Scale Alpha=&FloatFromCurve(
     FloatCurve=DataInterface() {propertyValues {Curve {keys [<at 0 0>,<at 0.1 1>]}}}
     Scale Curve=&Lerp_Float(
      A=&Multiply_Float(
       A=1
       B=&MergeHelper_FloatSwitch(
        0=&MakeFloatFromVector(Vector=$User.ShoeScale_R, Channel=::NewEnumerator0)
        1=&MakeFloatFromVector(Vector=$User.ShoeScale_L, Channel=::NewEnumerator0)
        2=?
        3=?
        Index=$Particles.PartitionParticles.Partition
        Num=::NewEnumerator0
       )
      )
      B=0
      Alpha=&MakeCustomFloatFromBool(Bool=$User.IsSkydiving, True Float=1, False Float=0)
     )
    )
    Scale RGB=&MakeVectorFromLinearColorRGB(Color=&MergeHelper_ColorSwitch(
      0=&FNTimeOfDayColor(Night Color=<rgba 0.1 0 0.0083 1>, Morning Color=<rgba 0.4 0 0 1>, Day Color=<rgba 0.4 0 0 1>, Evening Color=<rgba 0.4 0 0 1>)
      1=&FNTimeOfDayColor(Night Color=<rgba 0.25 0.2917 0.5 1>, Morning Color=<rgba 1 1 1 1>, Day Color=<rgba 1 1 1 1>, Evening Color=<rgba 1 1 1 1>)
      2=?
      3=?
      Index=$Particles.VisibilityTag
      Num=::NewEnumerator0
     ))
    ScaleA=true
    ScaleRGB=true
   )
   &AlignSpriteToMeshOrientation(SetAlignment=true, SetFacing=true)
   &SpriteRotationRate(Rotation Rate=&RandomRangeFloat(Minimum=-15, Maximum=15))
   &CameraOffset(Camera Offset Amount=&FloatFromCurve(
     FloatCurve=DataInterface() {propertyValues {Curve {
        defaultValue=1
        keys [{interpMode=RCIM_Cubic, tangentMode=RCTM_User, time=0, value=0}, {interpMode=RCIM_Cubic, tangentMode=RCTM_User, time=1, value=1}]
       }}}
     Scale Curve=10
    ))
  ]
  rendererClasses [&NiagaraSpriteRendererProperties]
  renderers [
   {rendererClass=&NiagaraSpriteRendererProperties}
   {rendererClass=&NiagaraSpriteRendererProperties, rendererIndex=1}
  ]
 }
 NE_Trail_01 {
  emitterSpawnScript [&FN_RibbonIDOffsetUtility()]
  emitterUpdateScript [
   &EmitterState()
   &FN_SpawnAnimTrail(Number of Anim Trails=2, Switch on ENiagara_AnimTrail_Source:ENiagara_AnimTrail_Source=::NewEnumerator1)
   /SetVariables(Emitter.Curve for Colors=DataInterface() {propertyValues {
      bExposeCurve=true
      AlphaCurve {keys [<at 0 1>,<at 0.5 1>,<at 1 1>]}
      BlueCurve {keys [<at 0 0.4452>,<at 0.5 0.5>,<at 1 0.4452>]}
      GreenCurve {keys [<at 0 0.4452>,<at 0.5 0.5>,<at 1 0.4452>]}
      RedCurve {keys [<at 0 0.4452>,<at 0.5 0.5>,<at 1 0.4452>]}
     }}) {isSetParameters=true, module=SetVariables_#5}
  ]
  particleSpawnScript [
   &FN_InitializeAnimTrailBoneIndecies(First Bone or Socket Index=4, Second Bone or Socket Index=5)
   &FN_InitializeAnimTrailBoneIndecies(AnimTrail Selection=1, First Bone or Socket Index=6, Second Bone or Socket Index=7) {module=FN_InitializeAnimTrailBoneIndecies001}
   &FN_InitializeAnimTrail(
    AnimTrailWidth=&Multiply_Float(
     A=8
     B=&MaskFloatBySpawnGroup(
      Masked Float=&MakeFloatFromVector(Vector=$User.ShoeScale_R, Channel=::NewEnumerator0)
      Passthrough Float=&MakeFloatFromVector(Vector=$User.ShoeScale_L, Channel=::NewEnumerator0)
     )
    )
    Color=&FNTimeOfDayColor(
     Night Color=<rgba 0.5 0.5417 1 1>
     Morning Color=<rgba 5 3.8764 3.23 1>
     Day Color=<rgba 2.2278 2.6503 5 1>
     Evening Color=<rgba 2.4126 2.2278 5 1>
    )
    Lifetime=0.35
    SkeletalMesh=$System.Skeletal Mesh
   )
   &AddVelocity(Velocity=&FN_FrontEndVector3Modifier(Gameplay=<0 0 0>, Front End=<0 -300 0>))
  ]
  particleUpdateScript [
   &SolveForcesAndVelocity(Clamp Velocity=false, Limit Acceleration=false)
   &FN_UpdateAnimTrail(Lock Particles=true, SkeletalMesh=$System.Skeletal Mesh) {module=FN_UpdateAnimTrail001}
   &DynamicMaterialParameters(
    Index 0 Param 1=1
    Index 1 Param 3=&FloatFromCurve(
     FloatCurve=DataInterface() {propertyValues {Curve {defaultValue=1, keys [{tangentMode=RCTM_User, time=0, value=0}, {tangentMode=RCTM_User, time=1, value=1}]}}}
     CurveIndex=&ModuloFloat(A=$Emitter.Age, B=1)
    )
    Index 1 Param 4=1
    Index 2 Param 1=10
    Index 2 Param 2=1
    Index 2 Param 3=0.5
    Index 2 Param 4=0.6
    Module.Index0Float0Write=true
    Module.Index0Float1Write=true
    Module.Index0Float2Write=true
    Module.Index0Float3Write=true
    Module.Index1Float0Write=true
    Module.Index1Float1Write=true
    Module.Index1Float2Write=true
    Module.Index1Float3Write=true
    Module.Index2Float0Write=true
    Module.Index2Float1Write=true
    Module.Index2Float2Write=true
    Module.Index2Float3Write=true
    Write Parameter Index 1=::NewEnumerator1
    Write Parameter Index 2=::NewEnumerator1
   )
   &ScaleColor(
    Scale Alpha=&FloatFromCurve(
     FloatCurve=DataInterface() {propertyValues {Curve {keys [
         {arriveTangent=-6.3309, interpMode=RCIM_Cubic, leaveTangent=-6.3309, tangentMode=RCTM_User, time=0, value=1}
         {interpMode=RCIM_Cubic, leaveTangent=-0.5849, tangentMode=RCTM_User, time=0.35, value=0}
        ]}}}
     Scale Curve=&MakeCustomFloatFromBool(
      Bool=&Boolean_AndOperation(Bool1=$System.bIsRunning, Bool2=&Boolean_NotOperation(Bool=$User.IsSkydiving))
      True Float=1
      False Float=0
     )
    )
    Scale RGB=&VectorFromFloat(Value=&FloatFromCurve(FloatCurve=DataInterface() {propertyValues {Curve {keys [<at 0 1>,<at 1 1>]}}}))
    ScaleA=true
    ScaleRGB=true
   )
   &ScaleRibbonWidth(Ribbon Width Scale=&FloatFromCurve(FloatCurve=DataInterface() {propertyValues {Curve {keys [<at 0 1>,<at 1 0.75>]}}}))
  ]
  rendererClasses [&NiagaraRibbonRendererProperties]
  renderers [{rendererClass=&NiagaraRibbonRendererProperties}]
 }
 NE_Trail_02 {
  emitterSpawnScript [&FN_RibbonIDOffsetUtility()]
  emitterUpdateScript [
   &EmitterState()
   &FN_SpawnAnimTrail(Number of Anim Trails=2, Switch on ENiagara_AnimTrail_Source:ENiagara_AnimTrail_Source=::NewEnumerator1)
   /SetVariables(Emitter.Curve for Colors=DataInterface() {propertyValues {
      bExposeCurve=true
      AlphaCurve {keys [<at 0 0>,<at 0.5 1>,<at 1 0>]}
      BlueCurve {keys [<at 0 0.4452>,<at 0.5 1>,<at 1 0.4452>]}
      GreenCurve {keys [<at 0 0.4452>,<at 0.5 1>,<at 1 0.4452>]}
      RedCurve {keys [<at 0 0.4452>,<at 0.5 1>,<at 1 0.4452>]}
     }}) {isSetParameters=true, module=SetVariables_#5}
  ]
  particleSpawnScript [
   &FN_InitializeAnimTrailBoneIndecies(First Bone or Socket Index=4, Second Bone or Socket Index=5)
   &FN_InitializeAnimTrailBoneIndecies(AnimTrail Selection=1, First Bone or Socket Index=6, Second Bone or Socket Index=7) {module=FN_InitializeAnimTrailBoneIndecies001}
   &FN_InitializeAnimTrail(
    AnimTrailWidth=&Multiply_Float(
     A=6
     B=&MaskFloatBySpawnGroup(
      Masked Float=&MakeFloatFromVector(Vector=$User.ShoeScale_R, Channel=::NewEnumerator0)
      Passthrough Float=&MakeFloatFromVector(Vector=$User.ShoeScale_L, Channel=::NewEnumerator0)
     )
    )
    Color=&MakeCustomLinearColorFromBool(Bool=$User.bIsFrontEndPreview, True Color=<rgba 0.3 0 0 1>, False Color=<rgba 0.3 0 0 0.555>)
    SkeletalMesh=$System.Skeletal Mesh
   )
   &AddVelocity(Velocity=&FN_FrontEndVector3Modifier(Gameplay=<0 0 0>, Front End=<0 -300 0>))
  ]
  particleUpdateScript [
   &SolveForcesAndVelocity(Clamp Velocity=false, Limit Acceleration=false)
   &FN_UpdateAnimTrail(Lock Particles=true, SkeletalMesh=$System.Skeletal Mesh) {module=FN_UpdateAnimTrail001}
   &DynamicMaterialParameters(
    Index 0 Param 1=1
    Index 1 Param 3=&FloatFromCurve(
     FloatCurve=DataInterface() {propertyValues {Curve {keys [<at 0 0>, {arriveTangent=1.9814, tangentMode=RCTM_User, time=1, value=1}]}}}
     CurveIndex=&ModuloFloat(A=$Emitter.Age, B=1)
     Scale Curve=0
    )
    Index 1 Param 4=1
    Index 2 Param 1=10
    Index 2 Param 2=1
    Index 2 Param 3=0.5
    Index 2 Param 4=0.6
    Module.Index0Float0Write=true
    Module.Index0Float1Write=true
    Module.Index0Float2Write=true
    Module.Index0Float3Write=true
    Module.Index1Float0Write=true
    Module.Index1Float1Write=true
    Module.Index1Float2Write=true
    Module.Index1Float3Write=true
    Module.Index2Float0Write=true
    Module.Index2Float1Write=true
    Module.Index2Float2Write=true
    Module.Index2Float3Write=true
    Write Parameter Index 1=::NewEnumerator1
    Write Parameter Index 2=::NewEnumerator1
   )
   &ScaleColor(
    Scale Alpha=&FloatFromCurve(
     FloatCurve=DataInterface() {propertyValues {Curve {keys [
         {arriveTangent=-6.3309, interpMode=RCIM_Cubic, leaveTangent=-6.3309, tangentMode=RCTM_User, time=0, value=1}
         {interpMode=RCIM_Cubic, leaveTangent=-0.5849, tangentMode=RCTM_User, time=0.35, value=0}
        ]}}}
     Scale Curve=&MakeCustomFloatFromBool(
      Bool=&Boolean_AndOperation(Bool1=$System.bIsRunning, Bool2=&Boolean_NotOperation(Bool=$User.IsSkydiving))
      True Float=&FNTimeOfDayScalar(Night Scale=0.5, Morning Scale=1, Evening Scale=1)
      False Float=0
     )
    )
    Scale RGB=&VectorFromFloat(Value=&FloatFromCurve(FloatCurve=DataInterface() {propertyValues {Curve {keys [<at 0 1>,<at 1 1>]}}}))
    ScaleA=true
    ScaleRGB=true
   )
   &ScaleRibbonWidth(Ribbon Width Scale=&FloatFromCurve(FloatCurve=DataInterface() {propertyValues {Curve {keys [<at 0 1>,<at 1 0.75>]}}}))
  ]
  rendererClasses [&NiagaraRibbonRendererProperties]
  renderers [{rendererClass=&NiagaraRibbonRendererProperties}]
 }
 NE_Trail_03 {
  emitterSpawnScript [&FN_RibbonIDOffsetUtility()]
  emitterUpdateScript [
   &EmitterState()
   &FN_SpawnAnimTrail(Number of Anim Trails=2, Switch on ENiagara_AnimTrail_Source:ENiagara_AnimTrail_Source=::NewEnumerator1)
   /SetVariables(Emitter.Curve for Colors=DataInterface() {propertyValues {
      bExposeCurve=true
      AlphaCurve {keys [<at 0 0>,<at 0.5 1>,<at 1 0>]}
      BlueCurve {keys [<at 0 0.4452>,<at 0.5 1>,<at 1 0.4452>]}
      GreenCurve {keys [<at 0 0.4452>,<at 0.5 1>,<at 1 0.4452>]}
      RedCurve {keys [<at 0 0.4452>,<at 0.5 1>,<at 1 0.4452>]}
     }}) {isSetParameters=true, module=SetVariables_#5}
  ]
  particleSpawnScript [
   &FN_InitializeAnimTrailBoneIndecies(First Bone or Socket Index=4, Second Bone or Socket Index=5)
   &FN_InitializeAnimTrailBoneIndecies(AnimTrail Selection=1, First Bone or Socket Index=6, Second Bone or Socket Index=7) {module=FN_InitializeAnimTrailBoneIndecies001}
   &FN_InitializeAnimTrail(
    AnimTrailWidth=&Multiply_Float(
     A=8
     B=&MaskFloatBySpawnGroup(
      Masked Float=&MakeFloatFromVector(Vector=$User.ShoeScale_R, Channel=::NewEnumerator0)
      Passthrough Float=&MakeFloatFromVector(Vector=$User.ShoeScale_L, Channel=::NewEnumerator0)
     )
    )
    Color=<rgba 0.15 0 0 1>
    Lifetime=0.4
    SkeletalMesh=$System.Skeletal Mesh
   )
   &AddVelocity(Velocity=&FN_FrontEndVector3Modifier(Gameplay=<0 0 0>, Front End=<0 -300 0>))
  ]
  particleUpdateScript [
   &SolveForcesAndVelocity(Clamp Velocity=false, Limit Acceleration=false)
   &FN_UpdateAnimTrail(Lock Particles=true, SkeletalMesh=$System.Skeletal Mesh) {module=FN_UpdateAnimTrail001}
   &DynamicMaterialParameters(
    Index 0 Param 1=1
    Index 1 Param 3=&FloatFromCurve(
     FloatCurve=DataInterface() {propertyValues {Curve {keys [<at 0 0>, {arriveTangent=1.9814, tangentMode=RCTM_User, time=1, value=1}]}}}
     CurveIndex=&ModuloFloat(A=$Emitter.Age, B=1)
     Scale Curve=0
    )
    Index 1 Param 4=1
    Index 2 Param 1=10
    Index 2 Param 2=1
    Index 2 Param 3=0.5
    Index 2 Param 4=0.6
    Module.Index0Float0Write=true
    Module.Index0Float1Write=true
    Module.Index0Float2Write=true
    Module.Index0Float3Write=true
    Module.Index1Float0Write=true
    Module.Index1Float1Write=true
    Module.Index1Float2Write=true
    Module.Index1Float3Write=true
    Module.Index2Float0Write=true
    Module.Index2Float1Write=true
    Module.Index2Float2Write=true
    Module.Index2Float3Write=true
    Write Parameter Index 1=::NewEnumerator1
    Write Parameter Index 2=::NewEnumerator1
   )
   &ScaleColor(
    Scale Alpha=&FloatFromCurve(
     FloatCurve=DataInterface() {propertyValues {Curve {keys [
         {arriveTangent=-6.3309, interpMode=RCIM_Cubic, leaveTangent=-6.3309, tangentMode=RCTM_User, time=0, value=1}
         {interpMode=RCIM_Cubic, leaveTangent=-0.5849, tangentMode=RCTM_User, time=0.35, value=0}
        ]}}}
     Scale Curve=&MakeCustomFloatFromBool(
      Bool=&Boolean_AndOperation(Bool1=$System.bIsRunning, Bool2=&Boolean_NotOperation(Bool=$User.IsSkydiving))
      True Float=&FNTimeOfDayScalar(Night Scale=0.5, Morning Scale=1, Day Scale=0.75, Evening Scale=1)
      False Float=0
     )
    )
    Scale RGB=&VectorFromFloat(Value=&FloatFromCurve(FloatCurve=DataInterface() {propertyValues {Curve {keys [<at 0 1>,<at 1 1>]}}}))
    ScaleA=true
    ScaleRGB=true
   )
   &ScaleRibbonWidth(Ribbon Width Scale=&FloatFromCurve(FloatCurve=DataInterface() {propertyValues {Curve {keys [<at 0 1>,<at 1 0.75>]}}}))
  ]
  rendererClasses [&NiagaraRibbonRendererProperties]
  renderers [{rendererClass=&NiagaraRibbonRendererProperties}]
 }
}
summary {
 systemName=NS_Shoe_ScarletBloom_Idle
 emitters [
  {emitterName=NE_Petals_Idle, rendererClasses [&NiagaraSpriteRendererProperties]}
  {emitterName=NE_Petals_Run, rendererClasses [&NiagaraSpriteRendererProperties]}
  {emitterName=NE_Trail_01, rendererClasses [&NiagaraRibbonRendererProperties]}
  {emitterName=NE_Trail_02, rendererClasses [&NiagaraRibbonRendererProperties]}
  {emitterName=NE_Trail_03, rendererClasses [&NiagaraRibbonRendererProperties]}
 ]
 userVariables [
  User.bIsFrontEnd {type=NiagaraBool}
  User.bIsFrontEndPreview {type=NiagaraBool}
  User.bisPrimaryPreviewItem {type=NiagaraBool}
  User.bIsSwinging {type=NiagaraBool}
  User.ForwardDot {type=NiagaraFloat, defaultValue=0, description=""}
  User.IsSkydiving {type=NiagaraBool}
  User.RibbonIDOffset {type=NiagaraInt32}
  User.ShoeScale_L {type=Vector3f, description="", defaultValue {struct=Vector3f, value=<1 1 1>}}
  User.ShoeScale_R {type=Vector3f, description="", defaultValue {struct=Vector3f, value=<0 0 0>}}
 ]
}
