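# Niagara landing effect for the Scarlet Bloom shoe cosmetic, combining short-lived vine meshes and petal bursts at both feet.
# Vines spawn at a decaying rate, pick random meshes, and follow skeletal bones or sockets with mirrored spin per foot.
# Two velocity-gated sprite bursts drift under noise, gravity and drag, fading out entirely while skydiving.
# Particle partitions apply left/right offsets and per-shoe scale parameters, and colors shift with time of day.
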NiagaraSystem /CosmeticShoes/Assets/FORT_Shoes/Shoe_ScarletBloom/FX/NS_Shoe_ScarletBloom_Landing
dependencies {
 usedDataInterfaces [
  &NiagaraDataInterfaceCurve
  &NiagaraDataInterfaceArrayFloat3
  &NiagaraDataInterfaceArrayInt32
  &NiagaraDataInterfaceArrayFloat
  &NiagaraDataInterfaceColorCurve
  &NiagaraDataInterfaceSkeletalMesh
 ]
 usedDynamicInputs [&MergeHelper_Vector2Switch, &VectorFromFloat]
 usedModules [
  &SystemState
  &self:SystemSpawnScript.SystemScriptSource.SystemScriptGraph.NiagaraNodeAssignment_0.SetVariables_#1
 ]
}
emitters {
 NE_Char_Vines {
  emitterUpdateScript [
   &EmitterState(Life Cycle Mode=::NewEnumerator1, Loop Behavior=::NewEnumerator1, Loop Duration=1, UseLoopDelay=false)
   &SpawnRate(
    SpawnRate=&FloatFromCurve(FloatCurve=DataInterface() {propertyValues {Curve {keys [<at 0 1>,<at 0.3 0>]}}}, CurveIndex=$Emitter.NormalizedLoopAge, Scale Curve=30)
    Use Spawn Probability=false
   )
  ]
  particleSpawnScript [
   /SetVariables(Particles.MeshIndex=&RandomRangeInt(Minimum=0, Maximum=3)) {isSetParameters=true}
   &PartitionParticles()
   &InitializeParticle(
    Color=&FNTimeOfDayColor(
     Night Color=<rgba 0.275 0.2979 0.55 1>
     Morning Color=<rgba 5 3.2403 2.2278 0.5>
     Day Color=<rgba 2.2278 2.6503 5 0.5>
     Evening Color=<rgba 3.0177 2.8696 5 0.5>
    )
    Lifetime=0.8
    Mesh Scale=&Multiply_Vector(
     A=<0.8 0.8 0.8>
     B=&MergeHelper_VectorSwitch(0=$User.ShoeScale_L, 1=$User.ShoeScale_R, 2=?, 3=?, Index=$Particles.PartitionParticles.Partition, Num=::NewEnumerator0)
    )
    Mesh Scale Mode=::NewEnumerator4
    UsePositionOffset=false
    Write Material Random=true
   )
   &InitialMeshOrientation(
    Orientation Axis=<0 -1 0>
    Orientation Vector=&SelectVectorFromArray(
     Direct Array Index=$Particles.PartitionParticles.Partition
     Vector Selection Array=DataInterface() {propertyValues {FloatData [<-1 0 0>, <1 0 0>]}}
    )
    Rotation Coordinate Space=::NewEnumerator2
    Use Orientation Vector=true
    Use Rotation Vector=false
   )
   &CameraOffset(Camera Offset Amount=10)
  ]
  particleUpdateScript [
   &ParticleState()
   &SkeletalMeshLocation(
    Bone / Socket Index=&SelectIntFromArray(Direct Array Index=$Particles.PartitionParticles.Partition, Int Selection Array=DataInterface() {propertyValues {IntData [0,1]}})
    MaskSpawnGroup=false
    Mesh Sampling Type=::NewEnumerator5
    Sampled Position Offset=&SelectVectorFromArray(
     Direct Array Index=$Particles.PartitionParticles.Partition
     Vector Selection Array=DataInterface() {propertyValues {FloatData [<-10 0 0>, <10 0 0>]}}
    )
    Skeletal Mesh=$System.Skeletal Mesh
    Socket Sampling Mode=::NewEnumerator5
   )
   &UpdateMeshOrientation(Rotation Rate=&FloatFromCurve(
     FloatCurve=DataInterface() {propertyValues {Curve {keys [<at 0 1>,<at 1 0>]}}}
     Scale Curve=&SelectFloatFromArray(Direct Array Index=$Particles.PartitionParticles.Partition, Float Selection Array=DataInterface() {propertyValues {FloatData [-15,15]}})
    ))
   &ScaleMeshSize(Scale Factor=&MakeVector(
     X=&FloatFromCurve(FloatCurve=DataInterface() {propertyValues {Curve {keys [<at 0 0>,<at 1 1>]}}}, Scale Curve=2)
     Y=&FloatFromCurve(FloatCurve=DataInterface() {propertyValues {Curve {keys [
          {interpMode=RCIM_Cubic, tangentMode=RCTM_User, time=0, value=0.7}
          {interpMode=RCIM_Cubic, tangentMode=RCTM_User, time=0.5, value=1}
          {arriveTangent=-4.4791, interpMode=RCIM_Cubic, leaveTangent=-4.4791, tangentMode=RCTM_User, time=1, value=0}
         ]}}})
     Z=&FloatFromCurve(FloatCurve=DataInterface() {propertyValues {Curve {keys [
          {interpMode=RCIM_Cubic, tangentMode=RCTM_User, time=0, value=0.7}
          {interpMode=RCIM_Cubic, tangentMode=RCTM_User, time=0.5, value=1}
          {arriveTangent=-4.4791, interpMode=RCIM_Cubic, leaveTangent=-4.4791, tangentMode=RCTM_User, time=1, value=0}
         ]}}})
    ))
   &ScaleColor(
    Linear Color Curve=DataInterface() {propertyValues {
      AlphaCurve {keys [<at 0 0>,<at 0.1221 1>,<at 1 0>]}
      BlueCurve {keys [<at 0 1>,<at 0.1141 5>,<at 0.5719 1>,<at 1 0>]}
      GreenCurve {keys [<at 0 0.85>,<at 0.1141 3.5>,<at 0.5719 0.85>,<at 1 0>]}
      RedCurve {keys [<at 0 0.8675>,<at 0.1141 3.675>,<at 0.5719 0.8675>,<at 1 0>]}
     }}
    Scale Mode=2
   )
  ]
  rendererClasses [&NiagaraMeshRendererProperties]
  renderers [{rendererClass=&NiagaraMeshRendererProperties, rendererIndex=0}]
 }
 NE_Petals_Burst {
  emitterUpdateScript [
   &EmitterState()
   &SpawnBurst_Instantaneous(
    Spawn Count=6
    Spawn Probability=&MakeCustomFloatFromBool(
     Bool=&SetBoolByFloatComparison(A=&MakeFloatFromVector(Vector=$Engine.Owner.Velocity, Channel=::NewEnumerator2), B=100)
     True Float=0
     False Float=1
    )
    Use Spawn Probability=true
    UseLoopCountLimit=false
   )
   &SpawnBurst_Instantaneous(
    Spawn Count=6
    Spawn Group=1
    Spawn Probability=&MakeCustomFloatFromBool(
     Bool=&SetBoolByFloatComparison(A=&MakeFloatFromVector(Vector=$Engine.Owner.Velocity, Channel=::NewEnumerator2), B=100)
     True Float=0
     False Float=1
    )
    Use Spawn Probability=true
    UseLoopCountLimit=false
   ) {module=SpawnBurst_Instantaneous002}
   /SetVariables(
    Emitter.Layer Color 1:LinearColor=&FNTimeOfDayColor(
     Night Color=<rgba 0.15 0 0.0112 1>
     Morning Color=<rgba 0.3 0 0.0224 1>
     Day Color=<rgba 0.3 0 0.0224 1>
     Evening Color=<rgba 0.3 0 0.0224 1>
    )
    Emitter.Layer Color 4:LinearColor=<rgba 0.0677 0.0009 0 1>
   ) {isSetParameters=true}
  ]
  particleSpawnScript [
   /SetVariables(Particles.VisibilityTag=&SpawnGroup()) {isSetParameters=true}
   &PartitionParticles()
   &InitializeParticle(
    Color=&MergeHelper_ColorSwitch(
     0=<rgba 0.2 0 0.0166 1>
     1=<rgba 1 1 1 1>
     2=<rgba 1 1 1 1>
     3=<rgba 1 1 1 1>
     Index=&SpawnGroup()
     Num=::NewEnumerator2
    )
    Lifetime Max=1
    Lifetime Min=0.5
    Lifetime Mode=::NewEnumerator1
    Position Offset=&MergeHelper_VectorSwitch(0=<20 20 5>, 1=<-20 -20 5>, 2=?, 3=?, Index=$Particles.PartitionParticles.Partition, Num=::NewEnumerator0)
    Position Offset Coordinate Space=::Local
    Sprite Rotation Mode=::NewEnumerator2
    Sprite Size Mode=::NewEnumerator3
    Uniform Sprite Size=&RandomRangeFloat(
     Minimum=&MergeHelper_FloatSwitch(0=6, 1=10, 2=?, 3=?, Index=$Particles.VisibilityTag, Num=::NewEnumerator0)
     Maximum=&MergeHelper_FloatSwitch(0=10, 1=12, 2=?, 3=?, Index=$Particles.VisibilityTag, Num=::NewEnumerator0)
    )
    UsePositionOffset=true
    Write Material Random=true
   )
   &AddVelocity(
    Cone Angle=90
    Cone Axis=<0 0 1>
    Module.UseConeFalloff=false
    Module.UseVelDistribution=false
    Override Seed=false
    Velocity Mode=::NewEnumerator2
    Velocity Speed=&RandomRangeFloat(Minimum=50, Maximum=200)
   )
   &ShapeLocation(MaskSpawnGroup=false, Override Seed=false, Sphere Radius=10, UseOwnerScale=true)
  ]
  particleUpdateScript [
   &ParticleState()
   &CurlNoiseForce(Bias Noise=false, Noise Frequency=5, Noise Strength=50, Pan Noise=false)
   &GravityForce(Gravity=&RandomRangeVector(Minimum=<0 0 -50>, Maximum=<0 0 -300>))
   &Drag(Drag=&RandomRangeFloat(Minimum=4, Maximum=8), Use Linear Drag=true, Use Rotational Drag=true)
   &InheritSourceMovement(
    Applied Position Scale=&RandomRangeVector(Minimum=<0 0 0>, Maximum=<0.5 0.5 0.5>)
    ApplyForce=false
    ApplyPosition=true
    ApplyPositionFalloff=false
   )
   &SolveForcesAndVelocity(Clamp Velocity=false, Limit Acceleration=false)
   &ScaleSpriteSize(Uniform Curve Sprite Scale=DataInterface() {propertyValues {Curve {defaultValue=0, keys [<at 0 1>, <at 0.8 1>, {tangentMode=RCTM_User, time=1, value=0}, {tangentMode=RCTM_User, time=1, value=0}]}}})
   &ScaleColor(
    Scale Alpha=&FloatFromCurve(
     FloatCurve=DataInterface() {propertyValues {Curve {keys [<at 0 0>,<at 0.1 1>]}}}
     Scale Curve=&Lerp_Float(
      A=&Multiply_Float(
       A=1
       B=&MergeHelper_FloatSwitch(
        0=&MakeFloatFromVector(Vector=$User.ShoeScale_L, Channel=::NewEnumerator0)
        1=&MakeFloatFromVector(Vector=$User.ShoeScale_R, Channel=::NewEnumerator0)
        2=?
        3=?
        Index=$Particles.PartitionParticles.Partition
        Num=::NewEnumerator0
       )
      )
      B=0
      Alpha=&MakeCustomFloatFromBool(Bool=$User.IsSkydiving, True Float=1, False Float=0)
     )
    )
    Scale RGB=&MakeVectorFromLinearColorRGB(Color=&MergeHelper_ColorSwitch(
      0=&FNTimeOfDayColor(Night Color=<rgba 0.1 0 0.0083 1>, Morning Color=<rgba 0.4 0 0 1>, Day Color=<rgba 0.4 0 0 1>, Evening Color=<rgba 0.4 0 0 1>)
      1=<rgba 1 1 1 1>
      2=<rgba 1 1 1 1>
      3=&FNTimeOfDayColor(Night Color=<rgba 0.25 0.2917 0.5 1>, Morning Color=<rgba 1 1 1 1>, Day Color=<rgba 1 1 1 1>, Evening Color=<rgba 1 1 1 1>)
      Index=$Particles.VisibilityTag
      Num=::NewEnumerator2
     ))
    ScaleA=true
    ScaleRGB=true
   )
   &AlignSpriteToMeshOrientation(SetAlignment=true, SetFacing=true)
   &SpriteRotationRate(Rotation Rate=&RandomRangeFloat(
     Minimum=&MergeHelper_FloatSwitch(0=-15, 1=0, 2=?, 3=?, Index=$Particles.VisibilityTag, Num=::NewEnumerator0)
     Maximum=&MergeHelper_FloatSwitch(0=15, 1=0, 2=?, 3=?, Index=$Particles.VisibilityTag, Num=::NewEnumerator0)
    ))
   &CameraOffset(Camera Offset Amount=&FloatFromCurve(
     FloatCurve=DataInterface() {propertyValues {Curve {
        defaultValue=1
        keys [{interpMode=RCIM_Cubic, tangentMode=RCTM_User, time=0, value=0}, {interpMode=RCIM_Cubic, tangentMode=RCTM_User, time=1, value=1}]
       }}}
     Scale Curve=10
    ))
   /SetVariables(Particles.FlowerDy:Vector4f=&MakeVector4(
     X=&FloatFromCurve(FloatCurve=DataInterface() {propertyValues {Curve {keys [<at 0 1>,<at 0.5 1>,<at 1 0>]}}})
     Y=0
     Z=0
     W=&FloatFromCurve(FloatCurve=DataInterface() {propertyValues {Curve {keys [<at 0 1>,<at 0.5 1>,<at 1 0>]}}})
    )) {isSetParameters=true}
  ]
  rendererClasses [&NiagaraSpriteRendererProperties]
  renderers [
   {rendererClass=&NiagaraSpriteRendererProperties}
   {rendererClass=&NiagaraSpriteRendererProperties, rendererIndex=1}
  ]
 }
}
summary {
 systemName=NS_Shoe_ScarletBloom_Landing
 emitters [
  {emitterName=NE_Petals_Burst, rendererClasses [&NiagaraSpriteRendererProperties]}
  {emitterName=NE_Char_Vines, rendererClasses [&NiagaraMeshRendererProperties]}
 ]
 userVariables [
  User.bIsFrontEnd {type=NiagaraBool}
  User.bIsFrontEndPreview {type=NiagaraBool}
  User.IsSkydiving {type=NiagaraBool}
  User.RibbonIDOffset {type=NiagaraInt32}
  User.ShoeScale_L {type=Vector3f, description="", defaultValue {struct=Vector3f, value=<1 1 1>}}
  User.ShoeScale_R {type=Vector3f, description="", defaultValue {struct=Vector3f, value=<0 0 0>}}
  User.UseGroundDecal {type=NiagaraBool, defaultValue=true}
 ]
}
